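# Deformation Control Rig for the M_MED_CrumpleFold male cosmetic body, correcting shoulders and placing zippers, collar, belt, grenades and straps as the body moves.
# A six-way sequence runs the shared DeformRig function and shoulder fixes, with extra upper-arm twist correction when isLOD2 is set.
# Later branches attach accessory bones through weighted parent constraints, then refine them with aim and position constraints, spherical pose readers, distance-driven offsets and spatial clamps against penetration.
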ControlRigBlueprint /BRCosmetics/Characters/Player/Male/Medium/Bodies/M_MED_CrumpleFold/Meshes/M_MED_CrumpleFold_Deform_ControlRig
graphs [
 RigVMFunctionLibrary {links [], nodes []}
 RigVMFunction_Sequence_2_ContainedGraph {nodes [
   n1 = RigVMFunctionEntryNode()
   RigVMFunction_Sequence = &RigVMFunction_Sequence(ExecuteContext=@n1.ExecuteContext)
   n2 = &RigVMFunction_Sequence(ExecuteContext=@RigVMFunction_Sequence.B)
   n3 = &RigVMFunction_Sequence(ExecuteContext=@n2.B)
   n4 = &RigVMFunction_Sequence(ExecuteContext=@n3.B)
   n5 = &RigVMFunction_Sequence(ExecuteContext=@n4.B)
   RigVMFunctionReturnNode(A=@RigVMFunction_Sequence.A, B=@n2.A, C=@n3.A, D=@n4.A, E=@n5.A, F=@n5.B)
  ]}
 RigVMModel {nodes [
   RigUnit_BeginExecution = &RigUnit_BeginExecution()
   RigVMFunction_Sequence = RigVMAggregateNode(ExecuteContext=@RigUnit_BeginExecution.ExecutePin)
   FN_DeformRig_FNC_1 = RigVMFunctionReferenceNode(
    DeformRig_ Curve="(deformRig_l_upperarm_up_alpha_72_#1=False,deformRig_r_upperarm_up_alpha_73_#2=False,deformRig_l_upperarm_fr_alpha_74_#3=False,deformRig_r_upperarm_fr_alpha_75_CB51B8 ...(truncated)"
    DeformRig_Bool="(upperarm_1_#4=true,elbow_3_#5=true,wrist_7_#6=true,groin_8_#7=True,glute_17_#8=True,knee_9_40C408324CA261 ...(truncated)"
    DeformRig_Debug=(upperarm_debug_18_#4=False,wrist_debug_20_#6=False,lat_debug_26_#9=False,groinGlute_debug_27_#7=False)
    DeformRig_Value="(upperarm_up_18_#10=(Rotation=(X=0.0,Y=-0.0,Z=0.0,W=1.0),Translation=(X=0.0,Y=3,Z=8.0),Scale3D=(X=0.70,Y=1.0,Z=1.0)),upperarm_fwd_fr_19_#11=(Rotation=(X=0.0,Y=-0.0,Z=0.0,W=1.0),Translation=(X=-10,Y= ...(truncated)"
    ExecuteContext=@RigVMFunction_Sequence.A
   )
   > n1 = &RigUnit_SphericalPoseReader(
    ActiveRegionScaleFactors=(PositiveWidth=2.5,NegativeWidth=3,PositiveHeight=3,NegativeHeight=3)
    ActiveRegionSize=0.1
    Debug=(bDrawDebug=false,DebugScale=25,DebugSegments=20,DebugThickness=0.25)
    DriverAxis=(X=1,Y=0,Z=0)
    DriverItem=(Type=Bone,Name="upperarm_r")
    FalloffRegionScaleFactors=(PositiveWidth=1,NegativeWidth=1,PositiveHeight=1,NegativeHeight=1)
    FalloffSize=0.7
    OptionalParentItem=(Type=Bone)
    RotationOffset=(X=0,Y=0,Z=0)
   )
   > &RigUnit_ModifyTransforms(
    ItemToModify="((Item=(Type=Bone,Name=\"deform_upperarm_bk_r\"),Transform=(Rotation=(X=0,Y=0,Z=0,W=1),Translation=(X=-2,Y=-3,Z=15),Scale3D=(X=1,Y=1,Z=1))),(Item=(Type=Bone,Name=\"upperarm_twist_02_r\"),Transform=(Rotation=(X=0,Y=0,Z=0,W=1),Translation=(X=0,Y=0,Z=0),Scale3D=( ...(truncated)"
    Weight=@n1.OutputParam
   )
   > n2 = &RigUnit_SphericalPoseReader(
    ActiveRegionScaleFactors=(PositiveWidth=2.5,NegativeWidth=3,PositiveHeight=3,NegativeHeight=3)
    ActiveRegionSize=0.1
    Debug=(bDrawDebug=false,DebugScale=25,DebugSegments=20,DebugThickness=0.25)
    DriverAxis=(X=1,Y=0,Z=0)
    DriverItem=(Type=Bone,Name="upperarm_l")
    FalloffRegionScaleFactors=(PositiveWidth=1,NegativeWidth=1,PositiveHeight=1,NegativeHeight=1)
    FalloffSize=0.7
    OptionalParentItem=(Type=Bone)
    RotationOffset=(X=0,Y=0,Z=0)
   )
   > n3 = &RigUnit_ModifyTransforms(
    ItemToModify="((Item=(Type=Bone,Name=\"deform_upperarm_bk_l\"),Transform=(Rotation=(X=0,Y=0,Z=0,W=1),Translation=(X=2,Y=3,Z=-12),Scale3D=(X=1,Y=1,Z=1))),(Item=(Type=Bone,Name=\"upperarm_twist_02_l\"),Transform=(Rotation=(X=0,Y=0,Z=0,W=1),Translation=(X=0,Y=0,Z=0),Scale3D=(X ...(truncated)"
    Weight=@n2.OutputParam
   )
   RigVMFunction_ControlFlowBranch = &RigVMFunction_ControlFlowBranch(ExecuteContext=@n3.ExecutePin, Condition=RigVMVariableNode() {variable=isLOD2}.Value)
   &RigUnit_ParentConstraint(
    AdvancedSettings=(InterpolationType=Average,RotationOrderForFilter=XZY)
    Child=(Type=Bone,Name="upperarm_twist_01_r")
    Parents=((Item=(Type=Bone,Name="upperarm_r"),Weight=1))
    Weight=1
    ExecutePin=@RigVMFunction_ControlFlowBranch.True
   )
   > &RigUnit_ParentConstraint(
    AdvancedSettings=(InterpolationType=Average,RotationOrderForFilter=XZY)
    Child=(Type=Bone,Name="upperarm_twist_01_l")
    Parents=((Item=(Type=Bone,Name="upperarm_l"),Weight=1))
    Weight=1
   )
   > n4 = &RigUnit_SphericalPoseReader(
    ActiveRegionScaleFactors=(PositiveWidth=2.5,NegativeWidth=3,PositiveHeight=3,NegativeHeight=3)
    ActiveRegionSize=0.1
    Debug=(bDrawDebug=false,DebugScale=25,DebugSegments=20,DebugThickness=0.25)
    DriverAxis=(X=1,Y=0,Z=0)
    DriverItem=(Type=Bone,Name="upperarm_r")
    FalloffRegionScaleFactors=(PositiveWidth=1,NegativeWidth=1,PositiveHeight=1,NegativeHeight=1)
    FalloffSize=0.7
    OptionalParentItem=(Type=Bone)
    RotationOffset=(X=0,Y=0,Z=0)
   )
   > &RigUnit_ModifyTransforms(
    ItemToModify=((Item=(Type=Bone,Name="upperarm_twist_01_r"),Transform=(Rotation=(X=-0.3285,Y=0.0285,Z=-0.2053,W=0.9215),Translation=(X=0,Y=0,Z=0),Scale3D=(X=1,Y=1,Z=1))))
    Weight=@n4.OutputParam
   )
   > n5 = &RigUnit_SphericalPoseReader(
    ActiveRegionScaleFactors=(PositiveWidth=2.5,NegativeWidth=3,PositiveHeight=3,NegativeHeight=3)
    ActiveRegionSize=0.1
    Debug=(bDrawDebug=false,DebugScale=25,DebugSegments=20,DebugThickness=0.25)
    DriverAxis=(X=1,Y=0,Z=0)
    DriverItem=(Type=Bone,Name="upperarm_l")
    FalloffRegionScaleFactors=(PositiveWidth=1,NegativeWidth=1,PositiveHeight=1,NegativeHeight=1)
    FalloffSize=0.7
    OptionalParentItem=(Type=Bone)
    RotationOffset=(X=0,Y=0,Z=0)
   )
   > &RigUnit_ModifyTransforms(
    ItemToModify=((Item=(Type=Bone,Name="upperarm_twist_01_l"),Transform=(Rotation=(X=-0.5274,Y=0,Z=0,W=0.8496),Translation=(X=0,Y=0,Z=0),Scale3D=(X=1,Y=1,Z=1))))
    Weight=@n5.OutputParam
   )
   &RigUnit_ParentConstraint(
    AdvancedSettings=(RotationOrderForFilter=XZY)
    Child=(Type=Bone,Name="zipper_bot_l_base")
    Parents="((Item=(Type=Bone,Name=\"deform_upperbelt\"),Weight=0.001),(Item=(Type=Bone,Name=\"deform_pec_l\"),Weight=0.0257),(Item=(Type=Bone,Name=\"clavicle_l\"),Weight=0.0514),(Item=(Type=Bone,Name=\"pelvis\"),Weight=0.0556),(Item=(Type=Bone,Name=\"spine_01\"),Weight=0.1171) ...(truncated)"
    Weight=1
    ExecutePin=@RigVMFunction_Sequence.B
   )
   > &RigUnit_ParentConstraint(
    AdvancedSettings=(RotationOrderForFilter=XZY)
    Child=(Type=Bone,Name="zipper_bot_r_base")
    Parents="((Item=(Type=Bone,Name=\"deform_pec_r\"),Weight=0.0086),(Item=(Type=Bone,Name=\"clavicle_r\"),Weight=0.0398),(Item=(Type=Bone,Name=\"pelvis\"),Weight=0.0494),(Item=(Type=Bone,Name=\"spine_01\"),Weight=0.1146),(Item=(Type=Bone,Name=\"spine_04\"),Weight=0.1161),(Item= ...(truncated)"
    Weight=1
   )
   > &RigUnit_ParentConstraint(
    AdvancedSettings=(RotationOrderForFilter=XZY)
    Child=(Type=Bone,Name="zipper_up_r_base")
    Parents="((Item=(Type=Bone,Name=\"upperarm_twist_01_r\"),Weight=0.0103),(Item=(Type=Bone,Name=\"deform_upperarm_fr_r\"),Weight=0.0155),(Item=(Type=Bone,Name=\"spine_02\"),Weight=0.0766),(Item=(Type=Bone,Name=\"clavicle_r\"),Weight=0.1065),(Item=(Type=Bone,Name=\"spine_05\"), ...(truncated)"
    Weight=1
   )
   > &RigUnit_ParentConstraint(
    AdvancedSettings=(RotationOrderForFilter=XZY)
    Child=(Type=Bone,Name="zipper_up_l_base")
    Parents="((Item=(Type=Bone,Name=\"upperarm_twist_01_l\"),Weight=0.0167),(Item=(Type=Bone,Name=\"deform_upperarm_fr_l\"),Weight=0.0222),(Item=(Type=Bone,Name=\"spine_02\"),Weight=0.071),(Item=(Type=Bone,Name=\"clavicle_l\"),Weight=0.1232),(Item=(Type=Bone,Name=\"spine_05\"),W ...(truncated)"
    Weight=1
   )
   > &RigUnit_ParentConstraint(
    AdvancedSettings=(RotationOrderForFilter=XZY)
    Child=(Type=Bone,Name="grenade_02_base")
    Parents=((Item=(Type=Bone,Name="thigh_l"),Weight=0.0002),(Item=(Type=Bone,Name="deform_glute_l"),Weight=0.1233),(Item=(Type=Bone,Name="pelvis"),Weight=0.3504),(Item=(Type=Bone,Name="thigh_twist_01_l"),Weight=0.5261))
    Weight=1
   )
   > &RigUnit_ParentConstraint(
    AdvancedSettings=(RotationOrderForFilter=XZY)
    Child=(Type=Bone,Name="grenade_01_base")
    Parents="((Item=(Type=Bone,Name=\"thigh_l\"),Weight=0.0033),(Item=(Type=Bone,Name=\"deform_groin_l\"),Weight=0.0053),(Item=(Type=Bone,Name=\"deform_glute_l\"),Weight=0.0465),(Item=(Type=Bone,Name=\"pelvis\"),Weight=0.3689),(Item=(Type=Bone,Name=\"thigh_twist_01_l\"),Weight=0 ...(truncated)"
    Weight=1
   )
   > &RigUnit_ParentConstraint(
    AdvancedSettings=(RotationOrderForFilter=XZY)
    Child=(Type=Bone,Name="strap_02_l_base")
    Parents="((Item=(Type=Bone,Name=\"spine_02\"),Weight=0.0001),(Item=(Type=Bone,Name=\"thigh_twist_01_r\"),Weight=0.0002),(Item=(Type=Bone,Name=\"thigh_twist_01_l\"),Weight=0.0021),(Item=(Type=Bone,Name=\"deform_groin_l\"),Weight=0.0021),(Item=(Type=Bone,Name=\"strap_base\"),W ...(truncated)"
    Weight=1
   )
   > &RigUnit_ParentConstraint(
    AdvancedSettings=(RotationOrderForFilter=XZY)
    Child=(Type=Bone,Name="strap_01_l_base")
    Parents="((Item=(Type=Bone,Name=\"deform_glute_l\"),Weight=0.0021),(Item=(Type=Bone,Name=\"spine_03\"),Weight=0.0119),(Item=(Type=Bone,Name=\"spine_01\"),Weight=0.0242),(Item=(Type=Bone,Name=\"spine_02\"),Weight=0.0284),(Item=(Type=Bone,Name=\"deform_upperbelt\"),Weight=0.20 ...(truncated)"
    Weight=1
   )
   > &RigUnit_ParentConstraint(
    AdvancedSettings=(RotationOrderForFilter=XZY)
    Child=(Type=Bone,Name="strap_01_r_base")
    Parents=((Item=(Type=Bone,Name="deform_glute_r"),Weight=0.0313),(Item=(Type=Bone,Name="deform_groin_r"),Weight=0.1124),(Item=(Type=Bone,Name="pelvis"),Weight=0.3505),(Item=(Type=Bone,Name="thigh_twist_01_r"),Weight=0.5057))
    Weight=1
   )
   > &RigUnit_ParentConstraint(
    AdvancedSettings=(RotationOrderForFilter=XZY)
    Child=(Type=Bone,Name="strap_02_r_base")
    Parents="((Item=(Type=Bone,Name=\"deform_groin_l\"),Weight=0.0007),(Item=(Type=Bone,Name=\"spine_01\"),Weight=0.0026),(Item=(Type=Bone,Name=\"thigh_twist_01_l\"),Weight=0.003),(Item=(Type=Bone,Name=\"spine_02\"),Weight=0.0036),(Item=(Type=Bone,Name=\"deform_upperbelt\"),Weig ...(truncated)"
    Weight=1
   )
   &RigUnit_AimConstraintLocalSpaceOffset(
    AimAxis=(X=0,Y=0,Z=1)
    Child=(Type=Bone,Name="knife_base")
    Filter=(bX=true,bY=false,bZ=false)
    Parents=((Item=(Type=Bone,Name="pelvis"),Weight=1))
    UpAxis=(X=1,Y=0,Z=0)
    Weight=1
    WorldUp=(Target=(X=0,Y=0,Z=1),Kind=Location,Space=(Type=Bone,Name="spine_01"))
    ExecutePin=@RigVMFunction_Sequence.C
   )
   > &RigUnit_AimConstraintLocalSpaceOffset(
    AimAxis=(X=-1,Y=0,Z=0)
    Child=(Type=Bone,Name="grenade_02_base")
    Filter=(bX=true,bY=false,bZ=false)
    Parents=((Item=(Type=Bone,Name="pelvis"),Weight=1))
    Weight=0.7
    WorldUp=(Target=(X=0,Y=0,Z=1),Kind=Location,Space=(Type=Bone,Name="spine_01"))
    bMaintainOffset=true
   )
   > &RigUnit_AimConstraintLocalSpaceOffset(
    AimAxis=(X=-1,Y=0,Z=0)
    Child=(Type=Bone,Name="grenade_01_base")
    Filter=(bX=true,bY=false,bZ=false)
    Parents=((Item=(Type=Bone,Name="pelvis"),Weight=1))
    Weight=0.7
    WorldUp=(Target=(X=0,Y=0,Z=1),Kind=Location,Space=(Type=Bone,Name="spine_01"))
    bMaintainOffset=true
   )
   > /Script/ControlRig.RigUnit_PositionConstraintLocalSpaceOffset(
    Child=(Type=Bone,Name="deform_upperbelt")
    Filter=(bX=true,bY=false,bZ=false)
    Parents=((Item=(Type=Bone,Name="deform_groin_l"),Weight=0.5),(Item=(Type=Bone,Name="deform_groin_r"),Weight=0.5))
    Weight=0.5
    bMaintainOffset=True
   ) {is=struct RigVMUnitNode}
   > &RigUnit_PositionConstraintLocalSpaceOffset(
    Child=(Type=Bone,Name="strap_02_l_base")
    Filter=(bX=true,bY=false,bZ=false)
    Parents=((Item=(Type=Bone,Name="deform_groin_l"),Weight=1))
    Weight=0.2
    bMaintainOffset=true
   )
   > &RigUnit_PositionConstraintLocalSpaceOffset(
    Child=(Type=Bone,Name="strap_base")
    Filter=(bX=true,bY=false,bZ=false)
    Parents=((Item=(Type=Bone,Name="deform_groin_l"),Weight=0.5),(Item=(Type=Bone,Name="deform_groin_r"),Weight=0.5))
    Weight=0.6
    bMaintainOffset=true
   )
   n6 = &RigVMFunction_MathDoubleRemap(
    SourceMaximum=36
    SourceMinimum=-1
    TargetMaximum=1
    TargetMinimum=0
    bClamp=False
    Value=&RigUnit_GetRelativeTransformForItem(Child=(Type=Bone,Name="deform_groin_l"), Parent=(Type=Bone,Name="deform_groin_l"), bChildInitial=true, bParentInitial=false).RelativeTransform.Translation.Y
   )
   n7 = &RigVMFunction_MathDoubleRemap(
    SourceMaximum=-36
    SourceMinimum=-6
    TargetMaximum=1
    TargetMinimum=0
    bClamp=False
    Value=&RigUnit_GetRelativeTransformForItem(Child=(Type=Bone,Name="deform_groin_r"), Parent=(Type=Bone,Name="deform_groin_r"), bChildInitial=true, bParentInitial=false).RelativeTransform.Translation.Y
   )
   > &RigUnit_ModifyTransforms(
    ItemToModify=((Item=(Type=Bone,Name="deform_upperbelt"),Transform=(Rotation=(X=0,Y=0,Z=0,W=1),Translation=(X=20,Y=4,Z=0),Scale3D=(X=1,Y=1,Z=1))))
    WeightMinimum=0
    Weight=&RigVMFunction_MathDoubleRemap(
     SourceMaximum=2
     SourceMinimum=0
     TargetMaximum=1
     TargetMinimum=0
     bClamp=False
     Value=&RigVMFunction_MathDoubleAdd(A=@n6.Result, B=@n7.Result).Result
    ).Result
   )
   > &RigUnit_ParentConstraint(
    AdvancedSettings=(RotationOrderForFilter=XZY)
    Child=(Type=Control,Name="deform_upperbelt_ctrl")
    Parents="((Item=(Type=Bone,Name=\"deform_groin_r\"),Weight=0.0078),(Item=(Type=Bone,Name=\"spine_01\"),Weight=0.0402),(Item=(Type=Bone,Name=\"spine_02\"),Weight=0.0468),(Item=(Type=Bone,Name=\"deform_upperbelt\"),Weight=0.3044),(Item=(Type=Bone,Name=\"pelvis\"),Weight=0.6008 ...(truncated)"
    Weight=1
    bMaintainOffset=true
   )
   RigUnit_GetTransform = &RigUnit_GetTransform(Item=(Type=Control,Name="spine_01_ctrl"), bInitial=False)
   RigVMFunction_MathVectorDistance = &RigVMFunction_MathVectorDistance(
    A=@RigUnit_GetTransform.Transform.Translation
    B=&RigUnit_GetTransform(Item=(Type=Bone,Name="ring_grenade_01_dyn"), bInitial=False).Transform.Translation
   )
   > &RigUnit_ModifyTransforms(
    ItemToModify="((Item=(Type=Bone,Name=\"grenade_01_base\"),Transform=(Rotation=(X=0,Y=0,Z=0,W=1),Translation=(X=2,Y=-2,Z=-8),Scale3D=(X=1,Y=1,Z=1))),(Item=(Type=Bone,Name=\"grenade_02_base\"),Transform=(Rotation=(X=0,Y=0,Z=0,W=1),Translation=(X=1,Y=-1,Z=-3),Scale3D=(X=1,Y=1, ...(truncated)"
    WeightMinimum=0
    Weight=&RigVMFunction_MathDoubleRemap(SourceMaximum=0, SourceMinimum=6, TargetMaximum=1, TargetMinimum=0, bClamp=False, Value=@RigVMFunction_MathVectorDistance.Result).Result
   )
   ClampSpatially = /Script/RigVM.RigVMFunction_MathTransformClampSpatially(
    Axis=Y
    DebugColor=(R=1,G=0,B=0,A=1)
    DebugThickness=1
    Maximum=50
    Minimum=8
    Type=Plane
    bDrawDebug=false
    Space=&RigUnit_GetTransform(Item=(Type=Bone,Name="thigh_twist_01_l"), bInitial=false).Transform
    Value=&RigUnit_GetTransform(Item=(Type=Bone,Name="strap_02_l_base"), bInitial=false).Transform
   ) {
    is=struct RigVMUnitNode
    template="ClampSpatially::Execute(in Value,in Axis,in Type,in Minimum,in Maximum,in Space,in bDrawDebug,in DebugColor,in DebugThickness,out Result)"
   }
   > Set Transform = /Script/ControlRig.RigUnit_SetTranslation(
    Item=(Type=Bone,Name="strap_02_l_base")
    Space=GlobalSpace
    Weight=1
    bInitial=false
    bPropagateToChildren=true
    Value=@ClampSpatially.Result.Translation
   ) {is=struct RigVMUnitNode, template="Set Transform::Execute(in Item,in Space,in bInitial,in Value,in Weight,in bPropagateToChildren)"}
   > Set Transform_3 = &RigUnit_SetTranslation(
    Item=(Type=Bone,Name="strap_02_l_base")
    Space=GlobalSpace
    Weight=1
    bInitial=false
    bPropagateToChildren=true
    Value=&RigVMFunction_MathTransformClampSpatially(
     Axis=X
     DebugColor=(R=1,G=0,B=0,A=1)
     DebugThickness=1
     Maximum=13
     Minimum=13
     Type=Cylinder
     bDrawDebug=false
     Space=&RigUnit_GetTransform(Item=(Type=Bone,Name="pelvis"), bInitial=false).Transform
     Value=&RigUnit_GetTransform(Item=(Type=Bone,Name="strap_02_l_base"), bInitial=false).Transform
    ) {template="ClampSpatially::Execute(in Value,in Axis,in Type,in Minimum,in Maximum,in Space,in bDrawDebug,in DebugColor,in DebugThickness,out Result)"}.Result.Translation
   ) {template="Set Transform::Execute(in Item,in Space,in bInitial,in Value,in Weight,in bPropagateToChildren)"}
   > Set Transform_2 = &RigUnit_SetTranslation(
    Item=(Type=Bone,Name="strap_base")
    Space=GlobalSpace
    Weight=1
    bInitial=false
    bPropagateToChildren=true
    Value=&RigVMFunction_MathTransformClampSpatially(
     Axis=X
     DebugColor=(R=1,G=0,B=0,A=1)
     DebugThickness=1
     Maximum=12
     Minimum=12
     Type=Cylinder
     bDrawDebug=false
     Space=&RigUnit_GetTransform(Item=(Type=Bone,Name="pelvis"), bInitial=false).Transform
     Value=&RigUnit_GetTransform(Item=(Type=Bone,Name="strap_base"), bInitial=false).Transform
    ) {template="ClampSpatially::Execute(in Value,in Axis,in Type,in Minimum,in Maximum,in Space,in bDrawDebug,in DebugColor,in DebugThickness,out Result)"}.Result.Translation
   ) {template="Set Transform::Execute(in Item,in Space,in bInitial,in Value,in Weight,in bPropagateToChildren)"}
   > Set Transform_4 = &RigUnit_SetTranslation(
    Item=(Type=Bone,Name="strap_base")
    Space=GlobalSpace
    Weight=1
    bInitial=false
    bPropagateToChildren=true
    Value=&RigVMFunction_MathTransformClampSpatially(
     Axis=X
     DebugColor=(R=1,G=0,B=0,A=1)
     DebugThickness=1
     Maximum=7
     Minimum=0
     Type=Plane
     bDrawDebug=false
     Space=&RigUnit_GetTransform(Item=(Type=Bone,Name="pelvis"), bInitial=false).Transform
     Value=&RigUnit_GetTransform(Item=(Type=Bone,Name="strap_base"), bInitial=false).Transform
    ) {template="ClampSpatially::Execute(in Value,in Axis,in Type,in Minimum,in Maximum,in Space,in bDrawDebug,in DebugColor,in DebugThickness,out Result)"}.Result.Translation
   ) {template="Set Transform::Execute(in Item,in Space,in bInitial,in Value,in Weight,in bPropagateToChildren)"}
   > Set Transform_5 = &RigUnit_SetTranslation(
    Item=(Type=Bone,Name="strap_02_l_base")
    Space=GlobalSpace
    bInitial=false
    bPropagateToChildren=true
    Value=&RigVMFunction_MathTransformClampSpatially(
     Axis=X
     DebugColor=(R=1,G=0,B=0,A=1)
     DebugThickness=1
     Maximum=50
     Minimum=5
     Type=Plane
     bDrawDebug=false
     Space=&RigUnit_GetTransform(Item=(Type=Bone,Name="pelvis"), bInitial=false).Transform
     Value=&RigUnit_GetTransform(Item=(Type=Bone,Name="strap_02_l_base"), bInitial=false).Transform
    ) {template="ClampSpatially::Execute(in Value,in Axis,in Type,in Minimum,in Maximum,in Space,in bDrawDebug,in DebugColor,in DebugThickness,out Result)"}.Result.Translation
    Weight=&RigVMFunction_MathDoubleRemap(
     SourceMaximum=36
     SourceMinimum=-1
     TargetMaximum=1
     TargetMinimum=0
     bClamp=False
     Value=&RigUnit_GetRelativeTransformForItem(Child=(Type=Bone,Name="deform_groin_l"), Parent=(Type=Bone,Name="deform_groin_l"), bChildInitial=true, bParentInitial=false).RelativeTransform.Translation.Y
    ).Result
   ) {template="Set Transform::Execute(in Item,in Space,in bInitial,in Value,in Weight,in bPropagateToChildren)"}
   > Set Transform_9 = &RigUnit_SetTranslation(
    Item=(Type=Bone,Name="strap_01_l_base")
    Space=GlobalSpace
    bInitial=false
    bPropagateToChildren=true
    Value=&RigVMFunction_MathTransformClampSpatially(
     Axis=X
     DebugColor=(R=1,G=0,B=0,A=1)
     DebugThickness=1
     Maximum=50
     Minimum=5
     Type=Plane
     bDrawDebug=false
     Space=&RigUnit_GetTransform(Item=(Type=Bone,Name="pelvis"), bInitial=false).Transform
     Value=&RigUnit_GetTransform(Item=(Type=Bone,Name="strap_01_l_base"), bInitial=false).Transform
    ) {template="ClampSpatially::Execute(in Value,in Axis,in Type,in Minimum,in Maximum,in Space,in bDrawDebug,in DebugColor,in DebugThickness,out Result)"}.Result.Translation
    Weight=&RigVMFunction_MathDoubleRemap(
     SourceMaximum=36
     SourceMinimum=-1
     TargetMaximum=1
     TargetMinimum=0
     bClamp=False
     Value=&RigUnit_GetRelativeTransformForItem(Child=(Type=Bone,Name="deform_groin_l"), Parent=(Type=Bone,Name="deform_groin_l"), bChildInitial=true, bParentInitial=false).RelativeTransform.Translation.Y
    ).Result
   ) {template="Set Transform::Execute(in Item,in Space,in bInitial,in Value,in Weight,in bPropagateToChildren)"}
   > Set Transform_6 = &RigUnit_SetTranslation(
    Item=(Type=Bone,Name="strap_02_l_base")
    Space=GlobalSpace
    Weight=0.07
    bInitial=false
    bPropagateToChildren=true
    Value=&RigVMFunction_MathTransformClampSpatially(
     Axis=X
     DebugColor=(R=1,G=0,B=0,A=1)
     DebugThickness=1
     Maximum=20
     Minimum=0
     Type=Plane
     bDrawDebug=false
     Space=&RigUnit_GetTransform(Item=(Type=Bone,Name="deform_upperbelt"), bInitial=false).Transform
     Value=&RigUnit_GetTransform(Item=(Type=Bone,Name="strap_02_l_base"), bInitial=false).Transform
    ) {template="ClampSpatially::Execute(in Value,in Axis,in Type,in Minimum,in Maximum,in Space,in bDrawDebug,in DebugColor,in DebugThickness,out Result)"}.Result.Translation
   ) {template="Set Transform::Execute(in Item,in Space,in bInitial,in Value,in Weight,in bPropagateToChildren)"}
   > Set Transform_8 = &RigUnit_SetTranslation(
    Item=(Type=Bone,Name="strap_01_l_base")
    Space=GlobalSpace
    Weight=1
    bInitial=false
    bPropagateToChildren=true
    Value=&RigVMFunction_MathTransformClampSpatially(
     Axis=X
     DebugColor=(R=1,G=0,B=0,A=1)
     DebugThickness=1
     Maximum=50
     Minimum=7.5
     Type=Cylinder
     bDrawDebug=false
     Space=&RigUnit_GetTransform(Item=(Type=Bone,Name="thigh_twist_01_l"), bInitial=false).Transform
     Value=&RigUnit_GetTransform(Item=(Type=Bone,Name="strap_01_l_base"), bInitial=false).Transform
    ) {template="ClampSpatially::Execute(in Value,in Axis,in Type,in Minimum,in Maximum,in Space,in bDrawDebug,in DebugColor,in DebugThickness,out Result)"}.Result.Translation
   ) {template="Set Transform::Execute(in Item,in Space,in bInitial,in Value,in Weight,in bPropagateToChildren)"}
   > Set Transform_1 = &RigUnit_SetTranslation(
    Item=(Type=Bone,Name="strap_01_l_base")
    Space=GlobalSpace
    Weight=1
    bInitial=false
    bPropagateToChildren=true
    Value=&RigVMFunction_MathTransformClampSpatially(
     Axis=X
     DebugColor=(R=1,G=0,B=0,A=1)
     DebugThickness=1
     Maximum=15
     Minimum=14
     Type=Cylinder
     bDrawDebug=false
     Space=&RigUnit_GetTransform(Item=(Type=Bone,Name="pelvis"), bInitial=false).Transform
     Value=&RigUnit_GetTransform(Item=(Type=Bone,Name="strap_01_l_base"), bInitial=false).Transform
    ) {template="ClampSpatially::Execute(in Value,in Axis,in Type,in Minimum,in Maximum,in Space,in bDrawDebug,in DebugColor,in DebugThickness,out Result)"}.Result.Translation
   ) {template="Set Transform::Execute(in Item,in Space,in bInitial,in Value,in Weight,in bPropagateToChildren)"}
   > Set Transform_7 = &RigUnit_SetTranslation(
    Item=(Type=Bone,Name="strap_base")
    Space=GlobalSpace
    Weight=0.1
    bInitial=false
    bPropagateToChildren=true
    Value=&RigVMFunction_MathTransformClampSpatially(
     Axis=X
     DebugColor=(R=1,G=0,B=0,A=1)
     DebugThickness=1
     Maximum=20
     Minimum=0
     Type=Plane
     bDrawDebug=false
     Space=&RigUnit_GetTransform(Item=(Type=Bone,Name="deform_upperbelt"), bInitial=false).Transform
     Value=&RigUnit_GetTransform(Item=(Type=Bone,Name="strap_base"), bInitial=false).Transform
    ) {template="ClampSpatially::Execute(in Value,in Axis,in Type,in Minimum,in Maximum,in Space,in bDrawDebug,in DebugColor,in DebugThickness,out Result)"}.Result.Translation
   ) {template="Set Transform::Execute(in Item,in Space,in bInitial,in Value,in Weight,in bPropagateToChildren)"}
   > n8 = &RigUnit_SphericalPoseReader(
    ActiveRegionScaleFactors=(PositiveWidth=3.5,NegativeWidth=0.5,PositiveHeight=3,NegativeHeight=3)
    ActiveRegionSize=0.1
    Debug=(bDrawDebug=false,DebugScale=25,DebugSegments=20,DebugThickness=0.25)
    DriverAxis=(X=-1,Y=0,Z=0)
    DriverItem=(Type=Bone,Name="thigh_l")
    FalloffRegionScaleFactors=(PositiveWidth=1,NegativeWidth=1,PositiveHeight=1,NegativeHeight=1)
    FalloffSize=0.45
    OptionalParentItem=(Type=Bone)
    RotationOffset=(X=90,Y=0,Z=-30)
   )
   > &RigUnit_ModifyTransforms(
    ItemToModify="((Item=(Type=Bone,Name=\"strap_02_l_base\"),Transform=(Rotation=(X=-0.173,Y=0.0858,Z=0.0151,W=0.9811),Translation=(X=0,Y=1,Z=0.5),Scale3D=(X=1,Y=1,Z=1))),(Item=(Type=Bone,Name=\"strap_01_l_base\"),Transform=(Rotation=(X=-0.0872,Y=0,Z=0,W=0.9962),Translation=(X ...(truncated)"
    WeightMinimum=0
    Weight=@n8.OutputParam
   )
   > &RigUnit_ParentConstraint(
    AdvancedSettings=(RotationOrderForFilter=XZY)
    Child=(Type=Bone,Name="deform_belt_01_base")
    Parents="((Item=(Type=Bone,Name=\"deform_groin_r\"),Weight=0.0078),(Item=(Type=Bone,Name=\"spine_01\"),Weight=0.0402),(Item=(Type=Bone,Name=\"spine_02\"),Weight=0.0468),(Item=(Type=Bone,Name=\"deform_upperbelt\"),Weight=0.3044),(Item=(Type=Bone,Name=\"pelvis\"),Weight=0.6008 ...(truncated)"
    Weight=1
   )
   n9 = &RigUnit_GetTransform(Item=(Type=Bone,Name="strap_base"), bInitial=False)
   n10 = &RigUnit_GetTransform(Item=(Type=Control,Name="deform_upperbelt_ctrl"), bInitial=False)
   > &RigUnit_ModifyTransforms(
    ItemToModify=((Item=(Type=Bone,Name="strap_base"),Transform=(Rotation=(X=0,Y=0,Z=0,W=1),Translation=(X=0,Y=0,Z=-3),Scale3D=(X=1,Y=1,Z=1))))
    WeightMinimum=0
    Weight=&RigVMFunction_MathDoubleRemap(
     SourceMaximum=3
     SourceMinimum=5.7
     TargetMaximum=1
     TargetMinimum=0
     Value=&RigVMFunction_MathVectorDistance(A=@n9.Transform.Translation, B=@n10.Transform.Translation).Result
    ).Result
   )
   > &RigUnit_AimConstraintLocalSpaceOffset(
    AimAxis=(X=0,Y=0,Z=-1)
    Child=(Type=Bone,Name="strap_01_l_base")
    Filter=(bX=true,bY=true,bZ=false)
    Parents=((Item=(Type=Bone,Name="strap_02_l_base"),Weight=1))
    UpAxis=(X=-1,Y=0,Z=0)
    Weight=1
    WorldUp=(Target=(X=0,Y=0,Z=1),Kind=Location,Space=(Type=Bone,Name="strap_01_l_base"))
    bMaintainOffset=true
   )
   > &RigUnit_AimConstraintLocalSpaceOffset(
    Child=(Type=Bone,Name="strap_base")
    Filter=(bX=false,bY=true,bZ=false)
    Parents=((Item=(Type=Bone,Name="strap_02_l_base"),Weight=1))
    UpAxis=(X=-1,Y=1,Z=0)
    Weight=1
    WorldUp=(Target=(X=0,Y=0,Z=1),Kind=Location,Space=(Type=Bone,Name="strap_base"))
    bMaintainOffset=true
   )
   > &RigUnit_AimConstraintLocalSpaceOffset(
    AimAxis=(X=0,Y=0,Z=-1)
    Child=(Type=Bone,Name="strap_02_l_base")
    Filter=(bX=false,bY=true,bZ=false)
    Parents=((Item=(Type=Bone,Name="strap_base"),Weight=1))
    UpAxis=(X=-1,Y=0,Z=0)
    Weight=1
    WorldUp=(Target=(X=0,Y=0,Z=1),Kind=Location,Space=(Type=Bone,Name="strap_02_l_base"))
    bMaintainOffset=true
   )
   n11 = &RigUnit_SphericalPoseReader(
    ActiveRegionScaleFactors=(PositiveWidth=1,NegativeWidth=1,PositiveHeight=1,NegativeHeight=1)
    ActiveRegionSize=0.1
    Debug=(bDrawDebug=false,DebugScale=25,DebugSegments=20,DebugThickness=0.25)
    DriverAxis=(X=1,Y=0,Z=0)
    DriverItem=(Type=Bone,Name="thigh_r")
    FalloffRegionScaleFactors=(PositiveWidth=1,NegativeWidth=0.4,PositiveHeight=1,NegativeHeight=0.5)
    FalloffSize=0.45
    FlipHeightScaling=false
    FlipWidthScaling=true
    OptionalParentItem=(Type=Bone)
    RotationOffset=(X=0,Y=0,Z=-40)
    ExecutePin=@RigVMFunction_Sequence.F
   )
   > &RigUnit_ModifyTransforms(
    ItemToModify="((Item=(Type=Bone,Name=\"strap_01_r_base\"),Transform=(Rotation=(X=-0.1361,Y=-0.3826,Z=-0.1018,W=0.9081),Translation=(X=-1.4997,Y=3.8577,Z=-4.6144),Scale3D=(X=1,Y=1,Z=1))),(Item=(Type=Bone,Name=\"strap_02_r_base\"),Transform=(Rotation=(X=0.5016,Y=-0.0655,Z=-0. ...(truncated)"
    WeightMinimum=0
    Weight=@n11.OutputParam
   )
   > n12 = &RigUnit_SphericalPoseReader(
    ActiveRegionScaleFactors=(PositiveWidth=1,NegativeWidth=1,PositiveHeight=1,NegativeHeight=1)
    ActiveRegionSize=0.1
    Debug=(bDrawDebug=false,DebugScale=25,DebugSegments=20,DebugThickness=0.25)
    DriverAxis=(X=-1,Y=0,Z=0)
    DriverItem=(Type=Bone,Name="thigh_l")
    FalloffRegionScaleFactors=(PositiveWidth=1,NegativeWidth=0.3,PositiveHeight=1,NegativeHeight=0.4)
    FalloffSize=0.45
    OptionalParentItem=(Type=Bone)
    RotationOffset=(X=180,Y=0,Z=-30)
   )
   > &RigUnit_ModifyTransforms(
    ItemToModify="((Item=(Type=Bone,Name=\"strap_01_l_base\"),Transform=(Rotation=(X=0.0076,Y=0.0554,Z=0.0238,W=0.9981),Translation=(X=0.8464,Y=2.059,Z=0.0221),Scale3D=(X=1,Y=1,Z=1))),(Item=(Type=Bone,Name=\"strap_02_l_base\"),Transform=(Rotation=(X=0,Y=-0.0211,Z=0.0001,W=0.999 ...(truncated)"
    WeightMinimum=0
    Weight=@n12.OutputParam
   )
   &RigUnit_AimConstraintLocalSpaceOffset(
    AimAxis=(X=0,Y=0,Z=-1)
    Child=(Type=Bone,Name="strap_02_r_base")
    Filter=(bX=true,bY=true,bZ=true)
    Parents=((Item=(Type=Bone,Name="calf_r"),Weight=1))
    UpAxis=(X=-1,Y=0,Z=0)
    Weight=1
    WorldUp=(Target=(X=0,Y=0,Z=1),Kind=Location,Space=(Type=Bone,Name="thigh_r"))
    bMaintainOffset=true
    ExecutePin=@RigVMFunction_Sequence.D
   )
   > &RigUnit_AimConstraintLocalSpaceOffset(
    AimAxis=(X=0,Y=0,Z=-1)
    Child=(Type=Bone,Name="strap_01_r_base")
    Filter=(bX=true,bY=true,bZ=true)
    Parents=((Item=(Type=Bone,Name="calf_r"),Weight=1))
    UpAxis=(X=-1,Y=0,Z=0)
    Weight=1
    WorldUp=(Target=(X=0,Y=0,Z=1),Kind=Location,Space=(Type=Bone,Name="thigh_r"))
    bMaintainOffset=true
   )
   > &RigUnit_AimConstraintLocalSpaceOffset(
    AimAxis=(X=0,Y=0,Z=-1)
    Child=(Type=Bone,Name="strap_02_r_base")
    Filter=(bX=false,bY=true,bZ=false)
    Parents=((Item=(Type=Bone,Name="strap_base"),Weight=1))
    UpAxis=(X=-1,Y=0,Z=0)
    Weight=1
    WorldUp=(Target=(X=0,Y=0,Z=1),Kind=Location,Space=(Type=Bone,Name="strap_02_r_base"))
    bMaintainOffset=true
   )
   > &RigUnit_AimConstraintLocalSpaceOffset(
    AimAxis=(X=-1,Y=0,Z=0)
    Child=(Type=Bone,Name="strap_base")
    Filter=(bX=false,bY=true,bZ=false)
    Parents=((Item=(Type=Bone,Name="strap_02_r_base"),Weight=1))
    UpAxis=(X=0,Y=1,Z=0)
    Weight=1
    WorldUp=(Target=(X=0,Y=0,Z=1),Kind=Location,Space=(Type=Bone,Name="strap_base"))
    bMaintainOffset=true
   )
   > &RigUnit_ParentConstraint(
    AdvancedSettings=(InterpolationType=Average,RotationOrderForFilter=XZY)
    Child=(Type=Bone,Name="strap_02_r_base")
    Parents=((Item=(Type=Bone,Name="deform_groin_r"),Weight=1))
    Weight=0.3
   )
   > Set Transform_10 = &RigUnit_SetTranslation(
    Item=(Type=Bone,Name="strap_01_r_base")
    Space=GlobalSpace
    Weight=1
    bInitial=false
    bPropagateToChildren=true
    Value=&RigVMFunction_MathTransformClampSpatially(
     Axis=X
     DebugColor=(R=1,G=0,B=0,A=1)
     DebugThickness=1
     Maximum=20
     Minimum=9.5
     Type=Cylinder
     bDrawDebug=false
     Space=&RigUnit_GetTransform(Item=(Type=Bone,Name="thigh_twist_01_r"), bInitial=false).Transform
     Value=&RigUnit_GetTransform(Item=(Type=Bone,Name="strap_01_r_base"), bInitial=false).Transform
    ) {template="ClampSpatially::Execute(in Value,in Axis,in Type,in Minimum,in Maximum,in Space,in bDrawDebug,in DebugColor,in DebugThickness,out Result)"}.Result.Translation
   ) {template="Set Transform::Execute(in Item,in Space,in bInitial,in Value,in Weight,in bPropagateToChildren)"}
   > Set Transform_11 = &RigUnit_SetTranslation(
    Item=(Type=Bone,Name="strap_02_r_base")
    Space=GlobalSpace
    Weight=0
    bInitial=false
    bPropagateToChildren=true
    Value=&RigVMFunction_MathTransformClampSpatially(
     Axis=X
     DebugColor=(R=1,G=0,B=0,A=1)
     DebugThickness=1
     Maximum=20
     Minimum=9.5
     Type=Cylinder
     bDrawDebug=false
     Space=&RigUnit_GetTransform(Item=(Type=Bone,Name="thigh_twist_01_r"), bInitial=false).Transform
     Value=&RigUnit_GetTransform(Item=(Type=Bone,Name="strap_02_r_base"), bInitial=false).Transform
    ) {template="ClampSpatially::Execute(in Value,in Axis,in Type,in Minimum,in Maximum,in Space,in bDrawDebug,in DebugColor,in DebugThickness,out Result)"}.Result.Translation
   ) {template="Set Transform::Execute(in Item,in Space,in bInitial,in Value,in Weight,in bPropagateToChildren)"}
   > n13 = &RigUnit_SphericalPoseReader(
    ActiveRegionScaleFactors=(PositiveWidth=3.5,NegativeWidth=0.5,PositiveHeight=3,NegativeHeight=3)
    ActiveRegionSize=0.1
    Debug=(bDrawDebug=false,DebugScale=25,DebugSegments=20,DebugThickness=0.25)
    DriverAxis=(X=1,Y=0,Z=0)
    DriverItem=(Type=Bone,Name="thigh_r")
    FalloffRegionScaleFactors=(PositiveWidth=1,NegativeWidth=1,PositiveHeight=1,NegativeHeight=1)
    FalloffSize=0.45
    FlipHeightScaling=false
    FlipWidthScaling=true
    OptionalParentItem=(Type=Bone)
    RotationOffset=(X=-90,Y=0,Z=-40)
   )
   > &RigUnit_ModifyTransforms(
    ItemToModify="((Item=(Type=Bone,Name=\"strap_02_r_base\"),Transform=(Rotation=(X=0.3194,Y=0.2501,Z=0.0016,W=0.914),Translation=(X=-1.567,Y=1.4333,Z=-4.7349),Scale3D=(X=1.1876,Y=1,Z=1))),(Item=(Type=Bone,Name=\"strap_01_r_base\"),Transform=(Rotation=(X=-0.4155,Y=0,Z=0,W=0.90 ...(truncated)"
    WeightMinimum=0
    Weight=@n13.OutputParam
   )
   Set Transform_12 = &RigUnit_SetTranslation(
    Item=(Type=Bone,Name="deform_collar_f")
    Space=GlobalSpace
    Weight=1
    bInitial=false
    bPropagateToChildren=true
    Value=&RigVMFunction_MathTransformClampSpatially(
     Axis=X
     DebugColor=(R=1,G=0,B=0,A=1)
     DebugThickness=1
     Maximum=20
     Minimum=8
     Type=Cylinder
     bDrawDebug=false
     Space=&RigUnit_GetTransform(Item=(Type=Bone,Name="neck_01"), bInitial=false).Transform
     Value=&RigUnit_GetTransform(Item=(Type=Bone,Name="deform_collar_f"), bInitial=false).Transform
    ) {template="ClampSpatially::Execute(in Value,in Axis,in Type,in Minimum,in Maximum,in Space,in bDrawDebug,in DebugColor,in DebugThickness,out Result)"}.Result.Translation
    ExecutePin=@RigVMFunction_Sequence.E
   ) {template="Set Transform::Execute(in Item,in Space,in bInitial,in Value,in Weight,in bPropagateToChildren)"}
   > Set Transform_13 = &RigUnit_SetTranslation(
    Item=(Type=Bone,Name="deform_collar_l")
    Space=GlobalSpace
    Weight=1
    bInitial=false
    bPropagateToChildren=true
    Value=&RigVMFunction_MathTransformClampSpatially(
     Axis=X
     DebugColor=(R=1,G=0,B=0,A=1)
     DebugThickness=1
     Maximum=20
     Minimum=8
     Type=Cylinder
     bDrawDebug=false
     Space=&RigUnit_GetTransform(Item=(Type=Bone,Name="neck_01"), bInitial=false).Transform
     Value=&RigUnit_GetTransform(Item=(Type=Bone,Name="deform_collar_l"), bInitial=false).Transform
    ) {template="ClampSpatially::Execute(in Value,in Axis,in Type,in Minimum,in Maximum,in Space,in bDrawDebug,in DebugColor,in DebugThickness,out Result)"}.Result.Translation
   ) {template="Set Transform::Execute(in Item,in Space,in bInitial,in Value,in Weight,in bPropagateToChildren)"}
   > Set Transform_14 = &RigUnit_SetTranslation(
    Item=(Type=Bone,Name="deform_collar_r")
    Space=GlobalSpace
    Weight=1
    bInitial=false
    bPropagateToChildren=true
    Value=&RigVMFunction_MathTransformClampSpatially(
     Axis=X
     DebugColor=(R=1,G=0,B=0,A=1)
     DebugThickness=1
     Maximum=20
     Minimum=8
     Type=Cylinder
     bDrawDebug=false
     Space=&RigUnit_GetTransform(Item=(Type=Bone,Name="neck_01"), bInitial=false).Transform
     Value=&RigUnit_GetTransform(Item=(Type=Bone,Name="deform_collar_r"), bInitial=false).Transform
    ) {template="ClampSpatially::Execute(in Value,in Axis,in Type,in Minimum,in Maximum,in Space,in bDrawDebug,in DebugColor,in DebugThickness,out Result)"}.Result.Translation
   ) {template="Set Transform::Execute(in Item,in Space,in bInitial,in Value,in Weight,in bPropagateToChildren)"}
  ]}
]
hierarchy [
 root {loc=<0 0 0>, rot=<0 0 0 1>}
 pelvis {parent=root, loc=<0 -0.7071 93.1191>, rot=<0.0127 0.707 -0.0127 -0.707>}
 spine_01 {parent=pelvis, loc=<6.7494 -0.6388 0>, rot=<0 0 0.034 -0.9994>}
 spine_02 {parent=spine_01, loc=<7.252 -0.035 0>, rot=<0 0 0.0238 -0.9997>}
 spine_03 {parent=spine_02, loc=<6.0106 0 0>, rot=<0 0 0.0149 -0.9999>}
 spine_04 {parent=spine_03, loc=<8.1024 0 0>, rot=<0 0 -0.0132 -0.9999>}
 spine_05 {parent=spine_04, loc=<10.0142 0 0>, rot=<0 0 -0.0015 -1>}
 attach_backpack {parent=spine_05, loc=<-0.1203 -13.4733 0>, rot=<-0.0093 -0.707 0.0093 -0.707>}
 neck_01 {parent=spine_05, loc=<11.3138 0.1529 0>, rot=<0 0 -0.2003 -0.9797>}
 neck_02 {parent=neck_01, loc=<4.5627 0 0>, rot=<0 0 0.0504 -0.9987>}
 head {parent=neck_02, loc=<5.3914 0 0>, rot=<0 0 0.115 -0.9934>}
 deform_collar_l {parent=neck_01, loc=<4.7311 1.6634 -6.0729>, rot=<0 0 0.0504 -0.9987>}
 deform_collar_r {parent=neck_01, loc=<4.7311 1.6634 6.0729>, rot=<0 0 0.0504 -0.9987>}
 deform_collar_f {parent=neck_01, loc=<3.6329 6.3551 0>, rot=<0 0 0.0504 -0.9987>}
 clavicle_l {parent=spine_05, loc=<4.3575 -0.1466 -2.3223>, rot=<0.0465 -0.7289 -0.009 -0.6829>}
 upperarm_l {parent=clavicle_l, loc=<15.5111 0 0>, rot=<0.0226 -0.3234 -0.0088 -0.9459>}
 lowerarm_l {parent=upperarm_l, loc=<26.6514 0 0>, rot=<0.0391 -0.0099 -0.2453 -0.9686>}
 hand_l {parent=lowerarm_l, loc=<22.3491 0 0>, rot=<-0.7081 -0.0507 0.0975 -0.6975>}
 thumb_01_l {parent=hand_l, loc=<1.3118 -0.3191 -2.459>, rot=<0.6546 -0.4042 0.0664 -0.6354>}
 thumb_02_l {parent=thumb_01_l, loc=<5.9336 -0.0448 0.0401>, rot=<-0.0037 0.0213 0.298 -0.9543>}
 thumb_03_l {parent=thumb_02_l, loc=<2.2223 0.3682 -0.0907>, rot=<0.0118 -0.013 0.0066 -0.9998>}
 index_metacarpal_l {parent=hand_l, loc=<1.6535 0.0183 -2.96>, rot=<0.0158 -0.0124 -0.0173 -0.9996>}
 index_01_l {parent=index_metacarpal_l, loc=<8.3008 0.0357 -0.0524>, rot=<-0.0044 0.0485 0.0924 -0.9945>}
 index_02_l {parent=index_01_l, loc=<3.3587 -0.3004 0.0452>, rot=<-0.002 0.0027 0.3225 -0.9466>}
 index_03_l {parent=index_02_l, loc=<2.5578 -0.123 -0.0252>, rot=<-0.002 0.0067 0.2558 -0.9667>}
 middle_metacarpal_l {parent=hand_l, loc=<2.1856 -0.1725 -0.5984>, rot=<0.006 0.0168 -0.0329 -0.9993>}
 middle_01_l {parent=middle_metacarpal_l, loc=<7.3275 -0.2447 0.0129>, rot=<-0.0208 0.0042 0.0858 -0.9961>}
 middle_02_l {parent=middle_01_l, loc=<4.0123 0.1401 0.1175>, rot=<-0.0218 0.0258 0.3512 -0.9357>}
 middle_03_l {parent=middle_02_l, loc=<2.6077 0.0534 0.1373>, rot=<0.0082 0.0188 0.1708 -0.9851>}
 ring_metacarpal_l {parent=hand_l, loc=<2.2074 -0.3625 1.3638>, rot=<-0.0284 0.0722 -0.0458 -0.9959>}
 ring_01_l {parent=ring_metacarpal_l, loc=<7.0295 -0.6197 -0.1446>, rot=<-0.0489 -0.0252 0.1892 -0.9804>}
 ring_02_l {parent=ring_01_l, loc=<3.6174 0.4647 -0.1027>, rot=<0 0 0.2806 -0.9598>}
 ring_03_l {parent=ring_02_l, loc=<2.4152 0.114 0.0199>, rot=<0 0 0.1189 -0.9929>}
 pinky_metacarpal_l {parent=hand_l, loc=<1.7616 -0.9753 3.3814>, rot=<-0.0397 0.0814 -0.069 -0.9935>}
 pinky_01_l {parent=pinky_metacarpal_l, loc=<7.1709 -0.4831 -0.1094>, rot=<-0.0654 -0.0035 0.2242 -0.9723>}
 pinky_02_l {parent=pinky_01_l, loc=<2.8274 0.0893 -0.1043>, rot=<0.0037 0.0011 0.2854 -0.9584>}
 pinky_03_l {parent=pinky_02_l, loc=<2.2046 0.0307 -0.0078>, rot=<-0.0002 0.0098 0.1445 -0.9895>}
 deform_wrist_dn_base_l {parent=hand_l, loc=<0 -2.2342 0>, rot=<0 0 0 -1>}
 deform_wrist_dn_l {parent=deform_wrist_dn_base_l, loc=<0 0 0>, rot=<0 0 0 1>}
 deform_wrist_up_base_l {parent=hand_l, loc=<0 2.2682 0>, rot=<0 0 0 -1>}
 deform_wrist_up_l {parent=deform_wrist_up_base_l, loc=<0 0 0>, rot=<0 0 0 1>}
 hand_attach_l {parent=hand_l, loc=<8.5 -3 0>, rot=<0.7071 0.7071 0 0>}
 weapon_l {parent=hand_l, loc=<-0.1974 -3.4111 0.3811>, rot=<0 0 0 1>}
 lowerarm_twist_02_l {parent=lowerarm_l, loc=<9.6199 0 0>, rot=<0 0 0 -1>}
 lowerarm_twist_01_l {parent=lowerarm_l, loc=<17.8549 0 0>, rot=<0 0 0 -1>}
 deform_elbow_in_base_l {parent=lowerarm_l, loc=<0.2112 5.259 0.8274>, rot=<0 0 0 -1>}
 deform_elbow_in_l {parent=deform_elbow_in_base_l, loc=<0 0 0>, rot=<0 0 0 1>}
 deform_elbow_out_base_l {parent=lowerarm_l, loc=<0 -4 0>, rot=<0 0 0 -1>}
 deform_elbow_out_l {parent=deform_elbow_out_base_l, loc=<0 0 0>, rot=<0 0 0 1>}
 deform_cuff_l {parent=lowerarm_l, loc=<9.6199 0 0>, rot=<0 0 0 -1>}
 dyn_cuff_l {parent=deform_cuff_l, loc=<0 0 0>, rot=<0 0 0 1>}
 upperarm_twist_01_l {parent=upperarm_l, loc=<0.0041 0 0>, rot=<0 0 0 -1>}
 deform_upperarm_bk_base_l {parent=upperarm_twist_01_l, loc=<1.0813 -7.2193 -1.9453>, rot=<0 0 0 1>}
 deform_upperarm_bk_l {parent=deform_upperarm_bk_base_l, loc=<0 0 0>, rot=<0 0 0 1>}
 deform_upperarm_fr_base_l {parent=upperarm_twist_01_l, loc=<-0.3691 7.0506 -2.765>, rot=<0 0 0 1>}
 deform_upperarm_fr_l {parent=deform_upperarm_fr_base_l, loc=<0 0 0>, rot=<0 0 0 1>}
 deform_upperarm_up_base_l {parent=upperarm_twist_01_l, loc=<-3.2216 -1.1391 3.3311>, rot=<0 0 0 1>}
 deform_upperarm_up_l {parent=deform_upperarm_up_base_l, loc=<0 0 0>, rot=<0 0 0 1>}
 upperarm_twist_02_l {parent=upperarm_l, loc=<13.7941 0 0>, rot=<0 0 0 -1>}
 attach_cape {parent=spine_05, loc=<9.0192 -11.7142 0>, rot=<0.0616 -0.7044 -0.0616 -0.7044>}
 clavicle_r {parent=spine_05, loc=<4.3575 -0.1466 2.3223>, rot=<0.7289 0.0465 -0.6829 0.009>}
 upperarm_r {parent=clavicle_r, loc=<-15.5111 0 0>, rot=<0.0226 -0.3234 -0.0088 -0.9459>}
 lowerarm_r {parent=upperarm_r, loc=<-26.6514 0 0>, rot=<0.0391 -0.0099 -0.2453 -0.9686>}
 hand_r {parent=lowerarm_r, loc=<-22.3491 0 0>, rot=<-0.7081 -0.0507 0.0975 -0.6975>}
 thumb_01_r {parent=hand_r, loc=<-1.3118 0.3191 2.459>, rot=<0.6546 -0.4042 0.0664 -0.6354>}
 thumb_02_r {parent=thumb_01_r, loc=<-5.9336 0.0448 -0.0401>, rot=<-0.0037 0.0213 0.298 -0.9543>}
 thumb_03_r {parent=thumb_02_r, loc=<-2.2223 -0.3682 0.0907>, rot=<0.0118 -0.013 0.0066 -0.9998>}
 index_metacarpal_r {parent=hand_r, loc=<-1.6535 -0.0183 2.96>, rot=<0.0158 -0.0124 -0.0173 -0.9996>}
 index_01_r {parent=index_metacarpal_r, loc=<-8.3008 -0.0357 0.0524>, rot=<-0.0044 0.0485 0.0924 -0.9945>}
 index_02_r {parent=index_01_r, loc=<-3.3587 0.3004 -0.0452>, rot=<-0.002 0.0027 0.3225 -0.9466>}
 index_03_r {parent=index_02_r, loc=<-2.5578 0.123 0.0252>, rot=<-0.002 0.0067 0.2558 -0.9667>}
 middle_metacarpal_r {parent=hand_r, loc=<-2.1856 0.1725 0.5984>, rot=<0.006 0.0168 -0.0329 -0.9993>}
 middle_01_r {parent=middle_metacarpal_r, loc=<-7.3275 0.2447 -0.0129>, rot=<-0.0208 0.0042 0.0858 -0.9961>}
 middle_02_r {parent=middle_01_r, loc=<-4.0123 -0.1401 -0.1175>, rot=<-0.0218 0.0258 0.3512 -0.9357>}
 middle_03_r {parent=middle_02_r, loc=<-2.6077 -0.0534 -0.1373>, rot=<0.0082 0.0188 0.1708 -0.9851>}
 ring_metacarpal_r {parent=hand_r, loc=<-2.2074 0.3625 -1.3638>, rot=<-0.0284 0.0722 -0.0458 -0.9959>}
 ring_01_r {parent=ring_metacarpal_r, loc=<-7.0295 0.6197 0.1446>, rot=<-0.0489 -0.0252 0.1892 -0.9804>}
 ring_02_r {parent=ring_01_r, loc=<-3.6174 -0.4647 0.1027>, rot=<0 0 0.2806 -0.9598>}
 ring_03_r {parent=ring_02_r, loc=<-2.4152 -0.114 -0.0199>, rot=<0 0 0.1189 -0.9929>}
 pinky_metacarpal_r {parent=hand_r, loc=<-1.7616 0.9753 -3.3814>, rot=<-0.0397 0.0814 -0.069 -0.9935>}
 pinky_01_r {parent=pinky_metacarpal_r, loc=<-7.1709 0.4831 0.1094>, rot=<-0.0654 -0.0035 0.2242 -0.9723>}
 pinky_02_r {parent=pinky_01_r, loc=<-2.8274 -0.0893 0.1043>, rot=<0.0037 0.0011 0.2854 -0.9584>}
 pinky_03_r {parent=pinky_02_r, loc=<-2.2046 -0.0307 0.0078>, rot=<-0.0002 0.0098 0.1445 -0.9895>}
 deform_wrist_dn_base_r {parent=hand_r, loc=<-0.0004 2.2345 -0.0001>, rot=<0 0 0 -1>}
 deform_wrist_dn_r {parent=deform_wrist_dn_base_r, loc=<0 0 0>, rot=<0 0 0 1>}
 deform_wrist_up_base_r {parent=hand_r, loc=<0.0003 -2.2684 0.0001>, rot=<0 0 0 -1>}
 deform_wrist_up_r {parent=deform_wrist_up_base_r, loc=<0 0 0>, rot=<0 0 0 1>}
 weapon_r {parent=hand_r, loc=<0.1974 3.4111 -0.3811>, rot=<0 0 0 1>}
 hand_attach_r {parent=hand_r, loc=<-8.5001 2.9999 0>, rot=<0 0 -0.7071 -0.7071>}
 lowerarm_twist_02_r {parent=lowerarm_r, loc=<-9.6199 0 0>, rot=<0 0 0 -1>}
 lowerarm_twist_01_r {parent=lowerarm_r, loc=<-17.8549 0 0>, rot=<0 0 0 -1>}
 deform_elbow_in_base_r {parent=lowerarm_r, loc=<-0.2111 -5.2591 -0.8275>, rot=<0 0 0 -1>}
 deform_elbow_in_r {parent=deform_elbow_in_base_r, loc=<-0.0001 0 0.0002>, rot=<0 0 0 1>}
 deform_elbow_out_base_r {parent=lowerarm_r, loc=<0 4 -0.0001>, rot=<0 0 0 -1>}
 deform_elbow_out_r {parent=deform_elbow_out_base_r, loc=<0 0 0>, rot=<0 0 0 1>}
 deform_cuff_r {parent=lowerarm_r, loc=<-9.6199 0 0>, rot=<0 -1 0 0>}
 dyn_cuff_r {parent=deform_cuff_r, loc=<0 0 0>, rot=<0 0 0 1>}
 upperarm_twist_01_r {parent=upperarm_r, loc=<-0.0041 0 0>, rot=<0 0 0 -1>}
 deform_upperarm_bk_base_r {parent=upperarm_twist_01_r, loc=<-1.0814 7.2192 1.9455>, rot=<0 0 0 1>}
 deform_upperarm_bk_r {parent=deform_upperarm_bk_base_r, loc=<0 0 0>, rot=<0 0 0 1>}
 deform_upperarm_fr_base_r {parent=upperarm_twist_01_r, loc=<0.3689 -7.0506 2.7652>, rot=<0 0 0 1>}
 deform_upperarm_fr_r {parent=deform_upperarm_fr_base_r, loc=<0 0 0>, rot=<0 0 0 1>}
 deform_upperarm_up_base_r {parent=upperarm_twist_01_r, loc=<3.2214 1.1391 -3.3308>, rot=<0 0 0 1>}
 deform_upperarm_up_r {parent=deform_upperarm_up_base_r, loc=<0 0 0>, rot=<0 0 0 -1>}
 upperarm_twist_02_r {parent=upperarm_r, loc=<-13.7941 0 0>, rot=<0 0 0 -1>}
 deform_pec_aim_base_l {parent=spine_04, loc=<7.8732 12.6388 -2.4123>, rot=<0.0102 -0.4602 0.3979 -0.7936>}
 deform_pec_aim_l {parent=deform_pec_aim_base_l, loc=<0 0 0>, rot=<0 0 0 1>}
 deform_pec_base_l {parent=deform_pec_aim_l, loc=<10.766 0 0>, rot=<0 0 0 1>}
 deform_pec_l {parent=deform_pec_base_l, loc=<0.0549 -0.0001 -0.0001>, rot=<0 0 0 1>}
 deform_pec_aim_base_r {parent=spine_04, loc=<7.8727 12.6388 2.4123>, rot=<-0.4602 -0.0102 0.7936 0.3979>}
 deform_pec_aim_r {parent=deform_pec_aim_base_r, loc=<0 0 0>, rot=<0 0 0 1>}
 deform_pec_base_r {parent=deform_pec_aim_r, loc=<-10.7662 -0.0006 -0.0007>, rot=<0 0 0 1>}
 deform_pec_r {parent=deform_pec_base_r, loc=<0 0.0001 0.0001>, rot=<0 0 0 1>}
 zipper_up_r_base {parent=spine_04, loc=<8.44 14.0148 6.353>, rot=<-0.1222 -0.1007 0.145 -0.9767>}
 zipper_up_r_dyn {parent=zipper_up_r_base, loc=<0 0 0>, rot=<0 0 0 1>}
 zipper_up_l_base {parent=spine_04, loc=<8.5524 13.1267 -8.0862>, rot=<0.1293 0.037 0.2095 -0.9685>}
 zipper_up_l_dyn {parent=zipper_up_l_base, loc=<0 0 0>, rot=<0 0 0 1>}
 zipper_bot_r_base {parent=spine_03, loc=<0.6913 12.0301 7.776>, rot=<-0.1984 0.0701 0.0122 -0.9775>}
 zipper_bot_r_dyn {parent=zipper_bot_r_base, loc=<0 0 0>, rot=<0 0 0 1>}
 zipper_bot_l_base {parent=spine_03, loc=<1.1764 10.6421 -9.3586>, rot=<0.2279 -0.0461 0.0655 -0.9704>}
 zipper_bot_l_dyn {parent=zipper_bot_l_base, loc=<0 0 0>, rot=<0 0 0 1>}
 spine_01_ctrl {controlType=EULER_TRANSFORM, parent=spine_01, type=Control, loc=<0 0 -13>, rot=<0 0 0 1>}
 thigh_l {parent=pelvis, loc=<-2.5693 -0.0109 -7.7501>, rot=<0.0012 0.0572 0.0155 -0.9982>}
 calf_l {parent=thigh_l, loc=<-39.0484 0 0>, rot=<0.0006 -0.0161 -0.0391 -0.9991>}
 foot_l {parent=calf_l, loc=<-42.561 0.1077 0.4292>, rot=<0.0418 -0.0455 0.0343 -0.9975>}
 ball_l {parent=foot_l, loc=<-6.1599 13.8264 0.6806>, rot=<0.0001 -0.0001 -0.7063 -0.7079>}
 calf_twist_02_l {parent=calf_l, loc=<-18.8064 -0.0967 0.0397>, rot=<0.0237 0.0095 -0.0035 -0.9997>}
 calf_twist_01_l {parent=calf_l, loc=<-35.813 -0.0967 0.0397>, rot=<0.0237 0.0095 -0.0035 -0.9997>}
 deform_knee_in_base_l {parent=calf_l, loc=<1.298 -5.2957 -0.0496>, rot=<0 0 0 -1>}
 deform_knee_in_l {parent=deform_knee_in_base_l, loc=<0 0 0>, rot=<0 0 0 1>}
 deform_knee_out_base_l {parent=calf_l, loc=<0 4 0>, rot=<0 0 0 -1>}
 deform_knee_out_l {parent=deform_knee_out_base_l, loc=<0 0 0>, rot=<0 0 0 1>}
 thigh_twist_01_l {parent=thigh_l, loc=<-21.6303 0 0>, rot=<0.0013 0 0 -1>}
 deform_groin_base_l {parent=thigh_l, loc=<-2.1908 8.1142 -1.7442>, rot=<0 0 0 -1>}
 deform_groin_l {parent=deform_groin_base_l, loc=<0 0 0>, rot=<0 0 0 1>}
 thigh_r {parent=pelvis, loc=<-2.5693 -0.0109 7.7501>, rot=<0.0572 -0.0012 0.9982 0.0155>}
 calf_r {parent=thigh_r, loc=<39.0484 0 0>, rot=<0.0006 -0.0161 -0.0391 -0.9991>}
 foot_r {parent=calf_r, loc=<42.561 -0.1077 -0.4292>, rot=<0.0418 -0.0455 0.0343 -0.9975>}
 ball_r {parent=foot_r, loc=<6.1599 -13.8264 -0.6806>, rot=<0.0001 -0.0001 -0.7063 -0.7079>}
 calf_twist_02_r {parent=calf_r, loc=<18.8064 0.0967 -0.0397>, rot=<0.0237 0.0095 -0.0035 -0.9997>}
 calf_twist_01_r {parent=calf_r, loc=<35.813 0.0967 -0.0397>, rot=<0.0237 0.0095 -0.0035 -0.9997>}
 deform_knee_in_base_r {parent=calf_r, loc=<-1.298 5.2957 0.0496>, rot=<0 0 0 -1>}
 deform_knee_in_r {parent=deform_knee_in_base_r, loc=<0 0 0>, rot=<0 0 0 1>}
 deform_knee_out_base_r {parent=calf_r, loc=<0 -4 0>, rot=<0 0 0 -1>}
 deform_knee_out_r {parent=deform_knee_out_base_r, loc=<0 0 0>, rot=<0 0 0 1>}
 thigh_twist_01_r {parent=thigh_r, loc=<21.6303 0 0>, rot=<0.0013 0 0 -1>}
 deform_groin_base_r {parent=thigh_r, loc=<2.1908 -8.1142 1.7442>, rot=<0 0 0 -1>}
 deform_groin_r {parent=deform_groin_base_r, loc=<0 0 0>, rot=<0 0 0 1>}
 deform_glute_base_l {parent=pelvis, loc=<-8.8064 -8.8222 -8.4287>, rot=<0.0275 0.0501 -0.4472 -0.8926>}
 deform_glute_l {parent=deform_glute_base_l, loc=<0 0 0>, rot=<0 0 0 1>}
 deform_glute_base_r {parent=pelvis, loc=<-8.8062 -8.8222 8.4287>, rot=<0.0501 -0.0275 0.8926 -0.4472>}
 deform_glute_r {parent=deform_glute_base_r, loc=<0 0 0>, rot=<0 0 0 1>}
 strap_01_r_base {parent=pelvis, loc=<-3.7733 0.7296 17.5349>, rot=<-0.0127 -0.707 0.0127 -0.707>}
 knife_base {parent=strap_01_r_base, loc=<-0.6514 0.2764 0.4674>, rot=<0.0187 0 -0.0022 -0.9998>}
 knife_01_dyn {parent=knife_base, loc=<-0.0171 0 0>, rot=<-0.0022 0 -0.0008 -1>}
 knife_02_dyn {parent=knife_01_dyn, loc=<-0.3124 0.0013 -1.1032>, rot=<0 0 0 1>}
 knife_base_ctrl {controlType=EULER_TRANSFORM, parent=knife_02_dyn, type=Control, loc=<0 0 10>, rot=<0 0 0 1>}
 strap_02_r_base {parent=pelvis, loc=<-2.7127 8.0792 11.6283>, rot=<-0.1687 -0.6288 -0.0824 -0.7546>}
 strap_base {parent=pelvis, loc=<-0.5374 10.9819 -0.67>, rot=<-0.0119 -0.6661 0.0134 -0.7456>}
 belt_01_base {parent=pelvis, loc=<4.5826 10.7099 2.2107>, rot=<0.2396 -0.4073 0.7431 0.4738>}
 belt_01_dyn {parent=belt_01_base, loc=<0 0 0>, rot=<0 0 0 1>}
 belt_02_dyn {parent=belt_01_dyn, loc=<4.6609 0 0>, rot=<0.016 -0.1173 -0.0976 -0.9882>}
 thighGrenade_l {parent=pelvis, loc=<-2.5693 -0.0109 -7.7501>, rot=<0.0012 0.0572 0.0155 -0.9982>}
 grenade_01_base {parent=thighGrenade_l, loc=<-2.7582 1.9455 -7.458>, rot=<-0.0586 -0.6668 -0.0791 -0.7387>}
 grenade_01_lookat {parent=grenade_01_base, loc=<0 0 0>, rot=<0 0 0 1>}
 grenade_01_dyn {parent=grenade_01_lookat, loc=<0 0 0>, rot=<0 -0.0606 0 -0.9982>}
 ring_grenade_01_dyn {parent=grenade_01_dyn, loc=<3.067 0.662 5.3239>, rot=<0.7741 -0.1277 -0.4611 -0.4145>}
 grenade_02_base {parent=thighGrenade_l, loc=<-2.224 -2.8752 -7.9137>, rot=<0.1404 -0.6407 0.0491 -0.7533>}
 grenade_02_lookat {parent=grenade_02_base, loc=<0 0 0>, rot=<0 0 0 1>}
 grenade_02_dyn {parent=grenade_02_lookat, loc=<0 0 0>, rot=<0 0 0 1>}
 ring_grenade_02_dyn {parent=grenade_02_dyn, loc=<3.0119 1.2236 4.3907>, rot=<0.6936 0.0075 -0.6697 -0.2653>}
 deform_upperbelt {parent=pelvis, loc=<4.3186 -0.473 0>, rot=<0 0 0.034 -0.9994>}
 deform_belt_01_base {parent=deform_upperbelt, loc=<-0.5594 11.2246 0.9214>, rot=<0.5595 0.2175 0.7952 0.0858>}
 dyn_belt_01 {parent=deform_belt_01_base, loc=<0 0 0>, rot=<0 0 0 1>}
 dyn_belt_02 {parent=dyn_belt_01, loc=<5.1979 0 0>, rot=<0.0078 -0.1151 0.1404 -0.9834>}
 deform_upperbelt_ctrl {controlType=EULER_TRANSFORM, parent=deform_upperbelt, type=Control, loc=<0 0 0>, rot=<0 0 0 1>, scale=<0.2 0.2 0.2>}
 strap_02_l_base {parent=pelvis, loc=<0.6288 8.5551 -9.3232>, rot=<0.089 -0.6449 0.2175 -0.7272>}
 strap_01_l_base {parent=pelvis, loc=<0.7209 4.1365 -14.2943>, rot=<0.283 -0.5673 0.3638 -0.6825>}
]
variables [
 isCrouching {type=bool, default=False}
 isLOD2 {type=bool, default=False}
 isSkydiveFloating {type=bool, default=False}
]
