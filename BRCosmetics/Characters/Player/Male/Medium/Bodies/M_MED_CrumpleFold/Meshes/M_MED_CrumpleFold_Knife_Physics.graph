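# Physics asset for the knife mesh of the male medium CrumpleFold player body, covering its knife, belt and thigh grenade accessories.
# Capsules on pelvis, spine, both thighs and right calf form the body, with thigh and calf joints locked.
# Knife and two-link belt chain use tight angular limits from their base bones.
# Grenade dyn bodies lock to lookat bodies, which rotate freely on bases under the left-thigh mount, itself constrained to the pelvis.
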PhysicsAsset /BRCosmetics/Characters/Player/Male/Medium/Bodies/M_MED_CrumpleFold/Meshes/M_MED_CrumpleFold_Knife_Physics
bodies {
 calf_r [
  {shapeType=Capsule, length=9.2851, radius=7.9156, center=<13.3783 2.5848 -0.1942>, rotation=<rot 0 -91.0465 -88.965>}
  {
   shapeType=Capsule
   length=24.9135
   radius=5.0865
   shapeName=calf_r_capsule
   center=<31.0831 1.2309 -0.5822>
   rotation=<rot 88.4387 -6.3105 0.0002>
  }
 ]
 deform_belt_01_base [{shapeType=Capsule, length=0.505, radius=0.505, shapeName=deform_belt_01_base_capsule, rotation=<rot 90 0 0>}]
 dyn_belt_01 [{shapeType=Capsule, length=2.6682, radius=2.167, shapeName=dyn_belt_01_capsule, center=<2.4862 0 0>, rotation=<rot 90 0 0>}]
 dyn_belt_02 [{shapeType=Capsule, length=2.5962, radius=2.263, shapeName=dyn_belt_02_capsule, center=<2.626 -1.2411 0>, rotation=<rot 90 0 0>}]
 grenade_01_base [{shapeType=Capsule, length=1, radius=1, shapeName=grenade_01_base_capsule}]
 grenade_01_dyn [{
   shapeType=Capsule
   length=5.5883
   radius=2.692
   shapeName=grenade_01_dyn_capsule
   center=<3.0996 -0.7468 0.8948>
   rotation=<rot 2.5988 -90 0.4853>
  }]
 grenade_01_lookat [{shapeType=Capsule, length=1, radius=1, shapeName=grenade_01_lookat_capsule}]
 grenade_02_base [{shapeType=Capsule, length=1, radius=1, shapeName=grenade_02_base_capsule}]
 grenade_02_dyn [{
   shapeType=Capsule
   length=5.5883
   radius=2.6556
   shapeName=grenade_02_dyn_capsule
   center=<2.8516 -1.2032 0.2114>
   rotation=<rot 4.3606 -90 -1.1536>
  }]
 grenade_02_lookat [{shapeType=Capsule, length=1, radius=1, shapeName=grenade_02_lookat_capsule}]
 knife_02_dyn [
  {
   shapeType=Capsule
   length=23.4677
   radius=4.6049
   shapeName=knife_02_dyn_capsule
   center=<-3.6123 -0.5644 -6.6357>
   rotation=<rot -2.7544 0 -0.0009>
  }
  {shapeType=Capsule, length=4.8005, radius=5.1296, shapeName=knife_02_dyn_capsule, center=<-3.4501 0 -7.8906>, rotation=<rot 0 0 90>}
 ]
 pelvis [{shapeType=Capsule, length=11.4445, radius=9.2675, center=<1.4419 1.8064 0.31>, rotation=<rot -1.576 -90 -0.7675>}]
 spine_01 [{shapeType=Capsule, length=10.2878, radius=10.7333, shapeName=spine_01_capsule, rotation=<rot 0 -90 0>}]
 strap_01_r_base [{shapeType=Capsule, length=1, radius=1, shapeName=strap_01_r_base_capsule}]
 strap_base [{
   shapeType=Capsule
   length=13.3889
   radius=3.9448
   shapeName=strap_base_capsule
   center=<-1.2798 -3.3612 -0.0192>
   rotation=<rot 0 86.0202 -89.7323>
  }]
 thighGrenade_l [{shapeType=Capsule, length=1, radius=1, shapeName=thighGrenade_l_capsule}]
 thigh_l [{shapeType=Capsule, length=36.4203, radius=9.3843, center=<-19.984 -0.2872 2.0823>, rotation=<rot 85.5878 0.0023 -179.993>}]
 thigh_r [{shapeType=Capsule, length=36.4203, radius=9.3843, center=<20.0338 -0.0015 -1.1241>, rotation=<rot 87.6215 0.0088 0.0138>}]
}
constraints [
 {
  bone1Name=thigh_r
  bone2Name=pelvis
  swing1LimitDegrees=45
  swing1Motion=Locked
  swing2LimitDegrees=45
  swing2Motion=Locked
  twistLimitDegrees=45
  twistMotion=Locked
 }
 {
  bone1Name=calf_r
  bone2Name=thigh_r
  swing1LimitDegrees=45
  swing1Motion=Locked
  swing2LimitDegrees=45
  swing2Motion=Locked
  twistLimitDegrees=45
  twistMotion=Locked
 }
 {bone1Name=knife_02_dyn, bone2Name=strap_01_r_base, swing1LimitDegrees=1, swing2LimitDegrees=10, twistLimitDegrees=45}
 {
  bone1Name=spine_01
  bone2Name=pelvis
  swing1LimitDegrees=45
  swing1Motion=Free
  swing2LimitDegrees=45
  swing2Motion=Free
  twistLimitDegrees=45
  twistMotion=Free
 }
 {
  bone1Name=strap_01_r_base
  bone2Name=pelvis
  swing1LimitDegrees=45
  swing1Motion=Free
  swing2LimitDegrees=45
  swing2Motion=Free
  twistLimitDegrees=45
  twistMotion=Free
 }
 {
  bone1Name=strap_base
  bone2Name=pelvis
  swing1LimitDegrees=45
  swing1Motion=Free
  swing2LimitDegrees=45
  swing2Motion=Free
  twistLimitDegrees=45
  twistMotion=Free
 }
 {
  bone1Name=thigh_l
  bone2Name=pelvis
  swing1LimitDegrees=45
  swing1Motion=Locked
  swing2LimitDegrees=45
  swing2Motion=Locked
  twistLimitDegrees=45
  twistMotion=Locked
 }
 {
  bone1Name=thighGrenade_l
  bone2Name=pelvis
  swing1LimitDegrees=45
  swing1Motion=Free
  swing2LimitDegrees=45
  swing2Motion=Free
  twistLimitDegrees=45
  twistMotion=Free
 }
 {
  bone1Name=grenade_01_base
  bone2Name=thighGrenade_l
  swing1LimitDegrees=45
  swing1Motion=Free
  swing2LimitDegrees=45
  swing2Motion=Free
  twistLimitDegrees=45
  twistMotion=Free
 }
 {
  bone1Name=grenade_01_lookat
  bone2Name=grenade_01_base
  swing1LimitDegrees=45
  swing1Motion=Free
  swing2LimitDegrees=45
  swing2Motion=Free
  twistLimitDegrees=45
  twistMotion=Free
 }
 {
  bone1Name=grenade_01_dyn
  bone2Name=grenade_01_lookat
  swing1LimitDegrees=45
  swing1Motion=Locked
  swing2LimitDegrees=45
  swing2Motion=Locked
  twistLimitDegrees=45
  twistMotion=Locked
 }
 {
  bone1Name=grenade_02_base
  bone2Name=thighGrenade_l
  swing1LimitDegrees=45
  swing1Motion=Free
  swing2LimitDegrees=45
  swing2Motion=Free
  twistLimitDegrees=45
  twistMotion=Free
 }
 {
  bone1Name=grenade_02_lookat
  bone2Name=grenade_02_base
  swing1LimitDegrees=45
  swing1Motion=Free
  swing2LimitDegrees=45
  swing2Motion=Free
  twistLimitDegrees=45
  twistMotion=Free
 }
 {
  bone1Name=grenade_02_dyn
  bone2Name=grenade_02_lookat
  swing1LimitDegrees=45
  swing1Motion=Locked
  swing2LimitDegrees=45
  swing2Motion=Locked
  twistLimitDegrees=45
  twistMotion=Locked
 }
 {
  bone1Name=deform_belt_01_base
  bone2Name=pelvis
  swing1LimitDegrees=45
  swing1Motion=Free
  swing2LimitDegrees=45
  swing2Motion=Free
  twistLimitDegrees=45
  twistMotion=Free
 }
 {bone1Name=dyn_belt_01, bone2Name=deform_belt_01_base, swing1LimitDegrees=25, swing2LimitDegrees=10, twistLimitDegrees=5}
 {bone1Name=dyn_belt_02, bone2Name=dyn_belt_01, swing1LimitDegrees=25, swing2LimitDegrees=10, twistLimitDegrees=5}
]
