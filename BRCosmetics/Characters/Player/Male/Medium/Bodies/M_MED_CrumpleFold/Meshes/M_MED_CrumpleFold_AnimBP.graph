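# Post-copy AnimBP for the CrumpleFold male medium body cosmetic, adding secondary motion to its accessories.
# The parent mesh pose feeds a deform Control Rig (crouch, skydive, LOD inputs), then zipper AnimDynamics, cached as 'Zippers'.
# That cache runs through a pelvis-space knife RigidBody, then grenade-ring and cuff AnimDynamics to output; an older grenade chain is unconnected.
# Update sets gravity override, LOD flag and knife alpha (zero while skydive floating).
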AnimBlueprint /BRCosmetics/Characters/Player/Male/Medium/Bodies/M_MED_CrumpleFold/Meshes/M_MED_CrumpleFold_AnimBP
target_skeleton=&/M_MED_CrumpleFold_Skeleton
graphs [
 AnimGraph {nodes [
   n1 = Anim Dynamics - Bone: grenade_01_base(Alpha=1, ExternalForce="0,0,0") {
    angularDampingOverride=1
    angularSpringConstant=20000
    bAngularSpring=true
    bOverrideAngularDamping=true
    bUseGravityOverride=true
    gravityScale=0.35
    simulationSpace=RootRelative
    alphaBoolBlend {blendOption=Linear}
    boundBone {boneName=grenade_01_base}
    componentLinearAccScale=<1 1 1>
    componentLinearVelScale=<1 1 1>
    gravityOverride {z=-980}
    physicsBodyDefinitions [{
      boundBone {boneName=grenade_01_base}
      boxExtents=<24 28 28>
      constraintSetup {angularLimitsMax=<15 15 15>, angularLimitsMin=<-15 -5 -15>, angularTarget=<1 0 0>}
      localJointOffset=<0 0 -4>
     }]
   }
   n2 = Anim Dynamics - Bone: grenade_02_base(ComponentPose=@n1, Alpha=1, ExternalForce="0,0,0") {
    angularDampingOverride=1
    angularSpringConstant=20000
    bAngularSpring=true
    bOverrideAngularDamping=true
    bUseGravityOverride=true
    gravityScale=0.35
    simulationSpace=RootRelative
    alphaBoolBlend {blendOption=Linear}
    boundBone {boneName=grenade_02_base}
    componentLinearAccScale=<1 1 1>
    componentLinearVelScale=<1 1 1>
    gravityOverride {z=-980}
    physicsBodyDefinitions [{
      boundBone {boneName=grenade_02_base}
      boxExtents=<30 35 35>
      constraintSetup {angularLimitsMax=<15 15 2>, angularLimitsMin=<-15 -5 -15>, angularTarget=<1 0 0>}
      localJointOffset=<0 0 -4>
     }]
   }
   n3 = Anim Dynamics - Bone: ring_grenade_01_dyn(ComponentPose=@n2, Alpha=1, ExternalForce="0,0,0") {
    angularDampingOverride=1
    angularSpringConstant=20000
    bAngularSpring=true
    bOverrideAngularDamping=true
    bUseGravityOverride=true
    gravityScale=0.35
    simulationSpace=RootRelative
    alphaBoolBlend {blendOption=Linear}
    boundBone {boneName=ring_grenade_01_dyn}
    componentLinearAccScale=<1 1 1>
    componentLinearVelScale=<1 1 1>
    gravityOverride {z=-980}
    physicsBodyDefinitions [{
      boundBone {boneName=ring_grenade_01_dyn}
      boxExtents=<20 25 25>
      constraintSetup {angularLimitsMax=<17 15 15>, angularLimitsMin=<-50 -15 -15>, angularTarget=<1 0 0>}
      localJointOffset=<4 0 0>
     }]
   }
   Anim Dynamics - Bone: ring_grenade_02_dyn(ComponentPose=@n3, Alpha=1, ExternalForce="0,0,0") {
    angularDampingOverride=1
    angularSpringConstant=20000
    bAngularSpring=true
    bOverrideAngularDamping=true
    bUseGravityOverride=true
    gravityScale=0.35
    simulationSpace=RootRelative
    alphaBoolBlend {blendOption=Linear}
    boundBone {boneName=ring_grenade_02_dyn}
    componentLinearAccScale=<1 1 1>
    componentLinearVelScale=<1 1 1>
    gravityOverride {z=-980}
    physicsBodyDefinitions [{
      boundBone {boneName=ring_grenade_02_dyn}
      boxExtents=<20 30 30>
      constraintSetup {angularLimitsMax=<15 15 15>, angularLimitsMin=<-15 -15 -15>, angularTarget=<1 0 0>}
      localJointOffset=<4 0 0>
     }]
   }
   n4 = RigidBody!(ComponentPose=Local To Component(LocalPose=Use cached pose 'Zippers'()), Alpha, OverrideWorldGravity:Vector="0,0,-980", ExternalForce) -> (Pose) {
    bClampLinearTranslationLimitToRefPose=true
    bOverrideWorldGravity=true
    overridePhysicsAsset=&/M_MED_CrumpleFold_Knife_Physics
    simulationSpace=BaseBoneSpace
    alphaBoolBlend {blendOption=Linear}
    baseBoneRef {boneName=pelvis}
    componentAppliedLinearAccClamp=<250 250 250>
    componentLinearAccScale=<0.4 0.4 0.4>
    componentLinearVelScale=<0.4 0.4 0.4>
    overrideWorldGravity {z=-980}
    simSpaceSettings {maxAngularAcceleration=300, maxAngularVelocity=150, maxLinearAcceleration=1000, maxLinearVelocity=100, worldAlpha=0.6032}
   }
   n5 = Anim Dynamics - Bone: ring_grenade_01_dyn!(ComponentPose=@n4, Alpha=1, GravityOverride:Vector="0,0,0") -> (Pose) {
    angularDampingOverride=0.99
    angularSpringConstant=150
    bAngularSpring=true
    bOverrideAngularDamping=true
    bUseGravityOverride=true
    simulationSpace=BoneRelative
    alphaBoolBlend {blendOption=Linear}
    boundBone {boneName=ring_grenade_01_dyn}
    componentAppliedLinearAccClamp=<3000 3000 3000>
    componentLinearAccScale=<0.8 0.8 0.8>
    componentLinearVelScale=<0.3 0.3 0.3>
    gravityOverride {x=-980}
    physicsBodyDefinitions [{
      boundBone {boneName=ring_grenade_01_dyn}
      boxExtents=<15 5 10>
      constraintSetup {angularLimitsMax=<0 15 5>, angularLimitsMin=<0 -15 -90>, angularTarget=<0 0 0>}
      localJointOffset=<2 0 0>
     }]
    relativeSpaceBone {boneName=root}
   }
   n6 = Anim Dynamics - Bone: ring_grenade_02_dyn!(ComponentPose=@n5, Alpha=1, GravityOverride:Vector="0,0,0") -> (Pose) {
    angularDampingOverride=0.99
    angularSpringConstant=150
    bAngularSpring=true
    bOverrideAngularDamping=true
    bUseGravityOverride=true
    simulationSpace=BoneRelative
    alphaBoolBlend {blendOption=Linear}
    boundBone {boneName=ring_grenade_02_dyn}
    componentAppliedLinearAccClamp=<3000 3000 3000>
    componentLinearAccScale=<0.8 0.8 0.8>
    componentLinearVelScale=<0.3 0.3 0.3>
    gravityOverride {x=-980}
    physicsBodyDefinitions [{
      boundBone {boneName=ring_grenade_02_dyn}
      boxExtents=<15 5 10>
      constraintSetup {angularLimitsMax=<0 15 5>, angularLimitsMin=<0 -15 -90>, angularTarget=<0 0 0>}
      localJointOffset=<2 0 0>
     }]
    relativeSpaceBone {boneName=root}
   }
   n7 = Anim Dynamics - Bone: dyn_cuff_l!(ComponentPose=@n6, Alpha=1, GravityOverride:Vector="1,1,1") -> (Pose) {
    angularDampingOverride=0.9999
    angularSpringConstant=1000
    bAngularSpring=true
    bLinearSpring=true
    bOverrideLinearDamping=true
    bUseGravityOverride=true
    linearDampingOverride=1
    linearSpringConstant=250
    alphaBoolBlend {blendOption=Linear}
    boundBone {boneName=dyn_cuff_l}
    componentAppliedLinearAccClamp=<200 200 200>
    componentLinearAccScale=<0.15 0.15 0.15>
    componentLinearVelScale=<0.15 0.15 0.15>
    gravityOverride=<1 1 1>
    physicsBodyDefinitions [{
      sphereCollisionRadius=0
      boundBone {boneName=dyn_cuff_l}
      boxExtents=<20 20 20>
      constraintSetup {angularLimitsMax=<5 5 0>, angularLimitsMin=<-5 -5 -15>, angularTarget=<1 0 0>, linearAxesMax=<3 3 3>, linearAxesMin=<-3 0 0>}
      localJointOffset=<4 0 0>
     }]
    relativeSpaceBone {boneName=root}
   }
   n8 = Anim Dynamics - Bone: dyn_cuff_r!(ComponentPose=@n7, Alpha=1, GravityOverride:Vector="1,1,1") -> (Pose) {
    angularDampingOverride=0.9999
    angularSpringConstant=1000
    bAngularSpring=true
    bLinearSpring=true
    bOverrideLinearDamping=true
    bUseGravityOverride=true
    linearDampingOverride=1
    linearSpringConstant=250
    alphaBoolBlend {blendOption=Linear}
    boundBone {boneName=dyn_cuff_r}
    componentAppliedLinearAccClamp=<200 200 200>
    componentLinearAccScale=<0.15 0.15 0.15>
    componentLinearVelScale=<0.15 0.15 0.15>
    gravityOverride=<1 1 1>
    physicsBodyDefinitions [{
      sphereCollisionRadius=0
      boundBone {boneName=dyn_cuff_r}
      boxExtents=<20 20 20>
      constraintSetup {angularLimitsMax=<5 5 0>, angularLimitsMin=<-5 -5 -15>, angularTarget=<1 0 0>, linearAxesMax=<3 0 3>, linearAxesMin=<-3 -3 0>}
      localJointOffset=<4 0 0>
     }]
    relativeSpaceBone {boneName=root}
   }
   Output Pose(Result=Component To Local(ComponentPose=@n8)) {layerGroup=None}
   n9 = Control Rig!(
    Source=Copy Pose From Mesh() {bUseAttachedParent=true}
    Alpha
    isSkydiveFloating:bool=Get bIsSkydiveFloating()
    isCrouching:bool=Get bIsCrouching()
    isLOD2:bool=Get LOD is Greater 1() -> (LOD is Greater 1:bool)
   ) -> (Pose) {
    alphaBoolBlend {blendOption=Linear}
    controlRigAssetReference {blueprintRigClass=&/M_MED_CrumpleFold_Deform_ControlRig.M_MED_CrumpleFold_Deform_ControlRig_C}
   }
   n10 = Anim Dynamics - Bone: zipper_up_r_dyn!(ComponentPose=Local To Component(LocalPose=@n9), Alpha=1, GravityOverride:Vector="0,0,0") -> (Pose) {
    angularDampingOverride=0.99
    angularSpringConstant=50
    bAngularSpring=true
    bGravityOverrideInSimSpace=true
    bOverrideAngularDamping=true
    bUseGravityOverride=true
    simulationSpace=BoneRelative
    alphaBoolBlend {blendOption=Linear}
    boundBone {boneName=zipper_up_r_dyn}
    componentAppliedLinearAccClamp=<3000 3000 3000>
    componentLinearAccScale=<0.8 0.8 0.8>
    componentLinearVelScale=<0.3 0.3 0.3>
    gravityOverride {x=-980}
    physicsBodyDefinitions [{
      boundBone {boneName=zipper_up_r_dyn}
      boxExtents=<15 5 10>
      constraintSetup {angularLimitsMax=<0 15 5>, angularLimitsMin=<0 -15 -90>, angularTarget=<0 0 0>}
      localJointOffset=<-1.8 0 0>
     }]
    relativeSpaceBone {boneName=root}
   }
   n11 = Anim Dynamics - Bone: zipper_up_l_dyn!(ComponentPose=@n10, Alpha=1, GravityOverride:Vector="0,0,0") -> (Pose) {
    angularDampingOverride=0.99
    angularSpringConstant=50
    bAngularSpring=true
    bGravityOverrideInSimSpace=true
    bOverrideAngularDamping=true
    bUseGravityOverride=true
    simulationSpace=BoneRelative
    alphaBoolBlend {blendOption=Linear}
    boundBone {boneName=zipper_up_l_dyn}
    componentAppliedLinearAccClamp=<3000 3000 3000>
    componentLinearAccScale=<0.8 0.8 0.8>
    componentLinearVelScale=<0.3 0.3 0.3>
    gravityOverride {x=-980}
    physicsBodyDefinitions [{
      boundBone {boneName=zipper_up_l_dyn}
      boxExtents=<15 5 10>
      constraintSetup {angularLimitsMax=<0 15 5>, angularLimitsMin=<0 -15 -90>, angularTarget=<0 0 0>}
      localJointOffset=<-1.5 0 0>
     }]
    relativeSpaceBone {boneName=root}
   }
   n12 = Anim Dynamics - Bone: zipper_bot_r_dyn!(ComponentPose=@n11, Alpha=1, GravityOverride:Vector="0,0,0") -> (Pose) {
    angularDampingOverride=0.99
    angularSpringConstant=50
    bAngularSpring=true
    bGravityOverrideInSimSpace=true
    bOverrideAngularDamping=true
    bUseGravityOverride=true
    simulationSpace=BoneRelative
    alphaBoolBlend {blendOption=Linear}
    boundBone {boneName=zipper_bot_r_dyn}
    componentAppliedLinearAccClamp=<3000 3000 3000>
    componentLinearAccScale=<0.8 0.8 0.8>
    componentLinearVelScale=<0.3 0.3 0.3>
    gravityOverride {x=-980}
    physicsBodyDefinitions [{
      boundBone {boneName=zipper_bot_r_dyn}
      boxExtents=<15 5 10>
      constraintSetup {angularLimitsMax=<0 15 5>, angularLimitsMin=<0 -15 -90>, angularTarget=<0 0 0>}
      localJointOffset=<-2 0 0>
     }]
    relativeSpaceBone {boneName=root}
   }
   n13 = Anim Dynamics - Bone: zipper_bot_l_dyn!(ComponentPose=@n12, Alpha=1, GravityOverride:Vector="0,0,0") -> (Pose) {
    angularDampingOverride=0.99
    angularSpringConstant=50
    bAngularSpring=true
    bGravityOverrideInSimSpace=true
    bOverrideAngularDamping=true
    bUseGravityOverride=true
    simulationSpace=BoneRelative
    alphaBoolBlend {blendOption=Linear}
    boundBone {boneName=zipper_bot_l_dyn}
    componentAppliedLinearAccClamp=<3000 3000 3000>
    componentLinearAccScale=<0.8 0.8 0.8>
    componentLinearVelScale=<0.3 0.3 0.3>
    gravityOverride {x=-980}
    physicsBodyDefinitions [{
      boundBone {boneName=zipper_bot_l_dyn}
      boxExtents=<15 5 10>
      constraintSetup {angularLimitsMax=<0 15 5>, angularLimitsMin=<0 -15 -90>, angularTarget=<0 0 0>}
      localJointOffset=<-2.2 0 0>
     }]
    relativeSpaceBone {boneName=root}
   }
   Save cached pose 'Zippers'(Pose=Component To Local(ComponentPose=@n13)) {cacheName=Zippers}
   Get bIsKneeSliding()
   Get alphaKnifeDynamics() -> (alphaKnifeDynamics:double)
  ]}
 EventGraph {nodes [
   Event BlueprintUpdateAnimation()
   > n1 = Gravity Override State Function DEPRECATED(CharacterPartAnimInstance=Self-Reference(), GameTimeInSeconds=GetGameTimeInSeconds())
   > Set Gravity Override(Gravity Override=@n1)
   > Set LOD is Greater 1(LOD is Greater 1:bool="integer >= integer"(A=Get CurrentLOD(), B=2)) -> (Output_Get:bool)
   > n2 = Branch(Condition=Get bIsSkydiveFloating())
   > Set alphaKnifeDynamics(alphaKnifeDynamics:double=0) -> (Output_Get:double)
   Set alphaKnifeDynamics(execute=@n2.else, alphaKnifeDynamics=1)
  ]}
]
variables [
 alphaKnifeDynamics {type=number}
 Gravity Override {type=Vector}
 LOD is Greater 1 {type=boolean}
]
